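# Translucent unlit two-sided FX material for the Celestial pickaxe's trails and particles, showing a tinted starfield and nebula.
# Panning screen-aligned star and nebula textures blend through smoke noise, scaled by a light mask or noise-distorted mask texture plus optional glow and edge falloff.
# Opacity combines that mask with noise dissolve, optionally driven by dynamic parameters, plus leading-edge gradient, particle alpha, depth, view-angle and camera-distance fades.
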Material /BRCosmetics/Weapons/FORT_Melee/Pickaxe_Celestial/FX/M_Pickaxe_Celestial
graph {
 properties {
  bUsedWithBeamTrails=true
  bUsedWithNiagaraRibbons=true
  bUsedWithNiagaraSprites=true
  bUsedWithParticleSprites=true
  blendMode=BLEND_Translucent
  pixelDepthOffsetMode=PDOM_Legacy
  shadingModel=MSM_Unlit
  twoSided=true
  usedShadingModels=Unlit
 }
 nodes [
  n1 = Constant()
  n2 = DynamicParameter() {defaultValue=<rgba 0.5 1 1 1>, paramNames [SphereMask_scale, Dissolve, Param3, Param4]}
  n3 = ScalarParameter() {parameterName=Use Dissolve DynParameter}
  n4 = TextureSampleParameter2D(UVs=Panner(Coordinate=TextureCoordinate() {uTiling=0.7, vTiling=0.8}) {speedX=0.2}) {parameterName=dissolution Texture, texture=&T_Inky_Smoke_Tile}
  n5 = TextureSample(UVs=Panner(Coordinate=TextureCoordinate() {uTiling=0.4, vTiling=0.5}) {speedY=0.25}) {texture=&T_NoisePack_01}
  n6 = SphereMask(
   A=ComponentMask(None=TextureCoordinate()) {g=true}
   B=Constant() {r=0.5}
   Radius=DynamicParameter() {defaultValue=<rgba 0.5 1 1 1>, paramNames [SphereMask_scale, Dissolve, Param3, Param4]}.SphereMask_scale
  ) {attenuationRadius=0.5, hardnessPercent=50}
  n7 = LinearInterpolate(
   A=Constant() {r=1}
   B=Multiply(A=Multiply(A=@n5.G, B=@n5.G), B=OneMinus(None=@n6))
   Alpha=ScalarParameter() {defaultValue=1, parameterName=Use Sphere Mask Dissolution}
  )
  n8 = Constant() {r=1}
  n9 = TextureCoordinate()
  n10 = ComponentMask(None=@n9) {r=true}
  n11 = Add(A=Abs(None=Subtract(A=@n10) {constB=0.5})) {constB=0.5}
  n12 = TextureSample(UVs=AppendVector(A=OneMinus(None=@n11), B=ComponentMask(None=@n9) {g=true})) {texture=/Game/Effects/Fort_Effects/Textures/Mask/T_BaseLightMask_Half}
  n13 = Panner() {speedY=-0.15}
  n14 = Multiply(A=ComponentMask(None=TextureSample(UVs=@n13) {texture=&T_NoisePack_01}.RGB) {g=true, r=true}) {constB=0.065}
  n15 = Multiply(
   A=TextureCoordinate()
   B=ComponentMask(None=VectorParameter() {parameterName=MaskTextureUVScale, defaultValue=<rgba 1 1 0 0>}.RGB) {g=true, r=true}
  )
  n16 = VectorParameter() {parameterName=MaskTexturePannerSpeed}
  n17 = TextureSampleParameter2D(UVs=Panner(Coordinate=Add(A=@n14, B=@n15), Speed=ComponentMask(None=@n16.RGB) {g=true, r=true})) {parameterName=MaskTexture, texture=/Game/Effects/Fort_Effects/BluGlow/ZoomNoise}
  n18 = Add(A=@n17.R, B=@n17.R)
  n19 = ScalarParameter() {parameterName=Use Alpha Mask Texture}
  n20 = Multiply(
   A=LinearInterpolate(A=@n12.R, B=@n18, Alpha=@n19)
   B=If(A=LinearInterpolate(A=@n1, B=@n2.Dissolve, Alpha=@n3), B=Multiply(A=@n4.R, B=@n7), "A > B"=@n8, "A == B"=@n8, "A < B"=Constant())
  )
  n21 = LinearInterpolate(
   A=Constant()
   B=TextureSample() {texture=/Game/Global/Textures/Default/Masks_and_Gradients/T_New_Soft_Flair}.R
   Alpha=ScalarParameter() {defaultValue=1, parameterName=Use Glow}
  )
  n22 = ComponentMask(None=TextureCoordinate()) {r=true}
  n23 = ScalarParameter() {parameterName=Leading Edge Contrast}
  n24 = Multiply(A=Multiply(A=Add(A=@n20, B=@n21), B=Power(Base=@n22, Exp=@n23)), B=ParticleColor().A)
  n25 = Multiply(
   A=ShadingPathSwitch(Default=DepthFade(Opacity=@n24), Mobile=@n24)
   B=MaterialFunctionCall("FadeSensitivity (optional)"=ScalarParameter() {parameterName=FadeSensitivity}) {materialFunction=&FadeBasedOnViewAngle}.Result
  )
  n26 = SphereMask(A=WorldPosition().XYZ, B=CameraPositionWS(), Radius=ScalarParameter() {parameterName=CameraFadeRadius}, Hardness=Constant())
  LinearInterpolate(A=@n25, B=Multiply(A=@n25, B=OneMinus(None=@n26)), Alpha=ScalarParameter() {parameterName=CameraFade})
  n27 = ScalarParameter() {defaultValue=128, parameterName=StarfieldTextureResolution}
  n28 = ComponentMask(None=VectorParameter() {parameterName="Nebula&&StarfieldPannerSpeed", defaultValue=<rgba 0.1 0 0 0>}.RGB) {g=true, r=true}
  n29 = TextureSample(UVs=Panner(
    Coordinate=MaterialFunctionCall!(TextureResolution=@n27) {materialFunction=/Engine/Functions/Engine_MaterialFunctions02/Texturing/ScreenAlignedPixelToPixelUVs}.Result
    Speed=@n28
   ) {speedX=0.1}) {samplerType=SAMPLERTYPE_Masks, texture=/Game/Characters/Player/Male/Medium/Bodies/M_Med_Soldier_04/Skins/Galaxy/Textures/Celestial_T_Stars}
  n30 = LinearInterpolate(A=Constant3Vector() {constant=<rgba 0.0015 0.0028 0.0104 1>}, B=Constant3Vector() {constant=<rgba 0.2135 0.0266 0.1457 1>}, Alpha=@n4.R)
  n31 = MaterialFunctionCall(TextureResolution=ScalarParameter() {defaultValue=256, parameterName=NebulaTextureResolution}) {materialFunction=&ScreenAlignedPixelToPixelUVs}
  n32 = Multiply(A=TextureSample(UVs=Panner(Coordinate=@n31.Result, Speed=@n28) {speedX=0.1}) {texture=/Game/Characters/Player/Male/Medium/Bodies/M_Med_Soldier_04/Skins/Galaxy/Textures/T_FN_Nebula}.RGB) {constB=3}
  n33 = Multiply(A=@n12.R, B=@n12.R)
  n34 = Multiply(
   A=Add(A=Multiply(A=@n29.R) {constB=0.85}, B=LinearInterpolate(A=@n30, B=@n32, Alpha=@n4.R))
   B=Add(A=LinearInterpolate(A=@n33, B=@n18, Alpha=@n19), B=@n21)
  )
  n35 = SphereMask(A=ComponentMask(None=TextureCoordinate()) {r=true}, B=Constant() {r=0.5}) {attenuationRadius=0.55, hardnessPercent=0}
  n36 = LinearInterpolate(
   A=@n34
   B=LinearInterpolate(A=Constant3Vector(), B=@n34, Alpha=Saturate(None=@n35))
   Alpha=ScalarParameter() {defaultValue=1, parameterName=Use Edge Power}
  )
  n37 = Constant3Vector() {constant=<rgba 0.8 1 1.5 1>}
  MaterialFunctionCall(Emissive=Multiply(A=VectorParameter() {parameterName=SpaceTint, defaultValue=<rgba 1 1 1 1>}.RGB, B=Multiply(A=@n36, B=@n37) {constB=1.4}) {constB=1.4}) {materialFunction=&MF_FX_EmissiveAdjustments}
 ]
}
