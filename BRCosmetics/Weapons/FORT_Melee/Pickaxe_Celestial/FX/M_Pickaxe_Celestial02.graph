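# Additive, unlit, two-sided glow material for the Celestial pickaxe effects, flagged for beam trails and Niagara ribbons.
# A blue-tinted, power-sharpened starburst is added to a soft flare masked toward low V, then multiplied by particle alpha.
# Both textures use center-scaled, clamped UVs, with constant-time Panners acting as fixed offsets rather than scrolling.
# Particle color feeds a shared emissive adjustment function separately.
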Material /BRCosmetics/Weapons/FORT_Melee/Pickaxe_Celestial/FX/M_Pickaxe_Celestial02
graph {
 properties {
  bUsedWithBeamTrails=true
  bUsedWithNiagaraRibbons=true
  blendMode=BLEND_Additive
  pixelDepthOffsetMode=PDOM_Legacy
  shadingModel=MSM_Unlit
  twoSided=true
  usedShadingModels=Unlit
 }
 nodes [
  n1 = ParticleColor()
  MaterialFunctionCall(Emissive=Multiply(A=Constant3Vector() {constant=<rgba 1 1 1 0>}, B=@n1.RGB)) {materialFunction=&MF_FX_EmissiveAdjustments}
  n2 = TextureCoordinate()
  n3 = Constant2Vector() {g=0.15, r=1.8}
  n4 = TextureSample(UVs=Panner(
    Coordinate=MaterialFunctionCall(UVs=@n2, Texture Scale=@n3) {materialFunction=&ScaleUVsByCenter}.UVs
    Time=Constant() {r=0.64}
   ) {speedY=1}) {samplerSource=SSM_Clamp_WorldGroupSettings, texture=/Game/Effects/Fort_Effects/Textures/Flares/T_Starburst}
  n5 = Multiply(A=Constant3Vector() {constant=<rgba 0.2135 0.6343 1 0>}, B=Multiply(A=Power(Base=@n4.R) {constExponent=5}) {constB=0.08})
  n6 = MaterialFunctionCall(UVs=TextureCoordinate(), Texture Scale=Constant2Vector() {g=0.35, r=0.68}) {materialFunction=&ScaleUVsByCenter}
  n7 = Constant() {r=0.295}
  n8 = MaterialFunctionCall(Float2=TextureCoordinate()) {materialFunction=&BreakOutFloat2Components}
  n9 = Multiply(
   A=TextureSample(UVs=Panner(Coordinate=@n6.UVs, Time=@n7) {speedY=1}) {samplerSource=SSM_Clamp_WorldGroupSettings, texture=/Game/Global/Textures/Default/Masks_and_Gradients/T_New_Soft_Flair}.R
   B=Power(Base=OneMinus(None=@n8.G)) {constExponent=8}
  )
  Multiply(A=@n1.A, B=Multiply(A=Add(A=@n5, B=@n9)) {constB=2})
 ]
}
