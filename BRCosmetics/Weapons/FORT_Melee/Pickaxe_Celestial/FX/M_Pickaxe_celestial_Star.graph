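# Additive unlit material for glowing star sprites on mesh particles, part of the Celestial pickaxe effects.
# World position offset builds camera-facing sprites sized by vertex color, shrunk near the camera, plus LOD-cull and black-hole offsets.
# Emissive layers a halo gradient, purple-tinted flare and hue-shifted blue-to-teal anime flare, pulsed by sine and tinted by particle color.
# The mobile shading path drops the halo and replaces hue shifting with a multiply.
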Material /BRCosmetics/Weapons/FORT_Melee/Pickaxe_Celestial/FX/M_Pickaxe_celestial_Star
graph {
 properties {
  bUsedWithMeshParticles=true
  blendMode=BLEND_Additive
  pixelDepthOffsetMode=PDOM_Legacy
  shadingModel=MSM_Unlit
  twoSided=true
  usedShadingModels=Unlit
 }
 nodes [
  n1 = Constant() {r=0.95}
  n2 = VertexColor()
  n3 = Constant() {r=0.48}
  n4 = Add(A=Add(A=Multiply(A=@n1, B=@n2.R), B=Multiply(A=@n3, B=@n2.G)), B=Multiply(A=@n2.B, B=Constant() {r=0.7}))
  n5 = ScalarParameter() {defaultValue=150, parameterName=NearCameraShrink}
  n6 = Multiply(
   A=ScalarParameter() {defaultValue=30, parameterName=SpriteScale}
   B=Power(Base=MaterialFunctionCall(Distance=@n5) {materialFunction=&Near_Camera_Fade}.Result) {constExponent=3}
  )
  n7 = VectorParameter() {parameterName=WorldPostionOfBlackHole, defaultValue=<rgba -30000 25536 1952 0>}
  n8 = AppendVector(A=@n7.RGB, B=@n7.RGB)
  n9 = MaterialFunctionCall!(
   World Position
   DistanceCullFadeAnimationOffset=MaterialFunctionCall() {materialFunction=/Game/Packages/Fortress_SharedMaterials/Base_Material_Functions/PickupEffectDistanceCullOverride}.Result
   Use actor center instead of pivot point
  ) {materialFunction=/Game/Packages/Fortress_SharedMaterials/Base_Material_Functions/LODEffect}
  n10 = Add(A=@n9.Result, B=WorldPosition().XYZ)
  Add(
   A=MaterialFunctionCall!(Center Location=WorldPosition().XYZ, 0-1 UVs, XY scale=Multiply(A=@n4, B=@n6), OverridePivot, "Normals (Optional)") {materialFunction=&Sprite}.World position offset
   B=Add(
    A=MaterialFunctionCall!(WorldPostionOfBlackHole=@n8, World Position=@n10) {materialFunction=/Game/Packages/Fortnite_UniqueMaterials_1/BlackHoleFunction}.Result
    B=@n9.Result
   )
  )
  n11 = Constant() {r=0.5}
  n12 = TextureSample(UVs=SphereMask(A=TextureCoordinate(), B=@n11, Radius=@n11, Hardness=Constant()) {hardnessPercent=25}) {texture=/Game/Effects/Fort_Effects/Textures/Grad/T_Halo_Corona_Gradient}
  n13 = Constant() {r=0.5}
  n14 = Constant3Vector() {constant=<rgba 0.506 0.3854 1 0>}
  n15 = TextureSample(UVs=TextureCoordinate()) {samplerType=SAMPLERTYPE_Grayscale, texture=/Game/Effects/Fort_Effects/Textures/T_EFX_CameraFlare_01}
  n16 = TextureSample() {samplerSource=SSM_Clamp_WorldGroupSettings, texture=/Game/Effects/Fort_Effects/Textures/Flares/T_Anime_Flare}
  n17 = LinearInterpolate(
   A=Constant3Vector() {constant=<rgba 0.1407 0.1494 1 0>}
   B=Constant3Vector() {constant=<rgba 0.0451 0.9427 0.7919 0>}
   Alpha=SphereMask(A=TextureCoordinate(), B=Constant() {r=0.5}, Radius=Constant() {r=0.5}) {hardnessPercent=0}
  )
  n18 = MaterialFunctionCall(Hue Shift Percentage=@n16.RGB, Texture=@n17) {materialFunction=&HueShift}
  n19 = Multiply(A=@n16.RGB, B=@n17)
  n20 = Add(
   A=ShadingPathSwitch(Default=Desaturation(None=@n12.RGB, Fraction=@n13), Mobile=Constant()) {desc=Migrated from FeatureLevelSwitch}
   B=Add(A=Multiply(A=@n14, B=@n15.R), B=ShadingPathSwitch(Default=@n18.Result, Mobile=@n19) {desc=Migrated from FeatureLevelSwitch})
  )
  n21 = LinearInterpolate(
   A=Constant() {r=0.1}
   B=Constant() {r=1}
   Alpha=MaterialFunctionCall(Speed=Constant() {r=0.5}) {materialFunction=&DebugTimeSine}.Result
  )
  n22 = ParticleColor()
  MaterialFunctionCall(Emissive=Multiply(A=Multiply(A=@n20, B=@n21), B=@n22.RGB)) {materialFunction=&MF_FX_EmissiveAdjustments}
  Multiply(A=@n22.A, B=Add(A=@n15.R, B=@n16.R))
 ]
}
