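# Additive, unlit, two-sided mesh-particle material for a glowing star effect on the Celestial pickaxe.
# Radially mapped, panning noise is multiplied by a center-scaled corona gradient, partly desaturated and tinted pale blue.
# That glow is boosted, multiplied by particle color and passed through the shared FX emissive adjustment function.
# A separate branch builds a soft UV edge mask from per-axis sphere masks, scaled by particle alpha.
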Material /BRCosmetics/Weapons/FORT_Melee/Pickaxe_Celestial/FX/M_Pickaxe_celestial_Star_1
graph {
 properties {
  bUsedWithMeshParticles=true
  blendMode=BLEND_Additive
  pixelDepthOffsetMode=PDOM_Legacy
  shadingModel=MSM_Unlit
  twoSided=true
  usedShadingModels=Unlit
 }
 nodes [
  n1 = TextureCoordinate() {uTiling=0.5, vTiling=0.5}
  n2 = TextureSample(UVs=Panner(Coordinate=MaterialFunctionCall(Vector or UVs=@n1) {materialFunction=&VectorToRadialValue}.Radial Coordinates) {speedY=-0.3}) {texture=&T_NoisePack_01}
  n3 = TextureSample(UVs=MaterialFunctionCall(UVs=TextureCoordinate(), Texture Scale=Constant2Vector() {g=1.6, r=1}) {materialFunction=&ScaleUVsByCenter}.UVs) {texture=/Game/Effects/Fort_Effects/Textures/Grad/T_Halo_Corona_Gradient}
  n4 = Desaturation(None=Multiply(A=@n2.R, B=Multiply(A=@n3.RGB) {constB=2}), Fraction=Constant() {r=0.3})
  FeatureLevelSwitch(Default=@n4, ES2=@n2.R, ES3_1=@n2.R, SM4_REMOVED=@n4)
  n5 = Constant() {r=2.2}
  n6 = Multiply(A=Constant3Vector() {constant=<rgba 0.3311 0.7431 0.984 0>}, B=@n4)
  n7 = ParticleColor()
  MaterialFunctionCall(Emissive=Multiply(A=Multiply(A=@n5, B=@n6), B=@n7.RGB)) {materialFunction=&MF_FX_EmissiveAdjustments}
  n8 = TextureCoordinate()
  n9 = Saturate(None=SphereMask(A=ComponentMask(None=@n8) {r=true}, B=Constant() {r=0.5}) {attenuationRadius=0.5, hardnessPercent=0})
  n10 = OneMinus(None=@n9)
  n11 = ComponentMask(None=TextureCoordinate()) {g=true}
  n12 = Constant() {r=0.5}
  n13 = OneMinus(None=Saturate(None=SphereMask(A=@n11, B=@n12) {attenuationRadius=0.9, hardnessPercent=0}))
  Multiply(A=@n7.A, B=Multiply(A=@n9, B=Add(A=@n10, B=@n13)))
 ]
}
